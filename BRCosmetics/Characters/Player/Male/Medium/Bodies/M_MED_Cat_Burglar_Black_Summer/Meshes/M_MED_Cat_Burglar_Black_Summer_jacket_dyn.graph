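# Physics asset for the jacket of the male medium Cat Burglar Black Summer body, setting up dynamic motion of its hanging panels.
# Capsule chains cover the back, both sides and two-segment left and right fronts, each starting at a small MSTR body constrained to the pelvis with free rotation.
# Tight swing and twist limits shape each panel, with a much wider second swing on the fronts.
# Pelvis and thigh capsules provide body volumes around the jacket.
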PhysicsAsset /BRCosmetics/Characters/Player/Male/Medium/Bodies/M_MED_Cat_Burglar_Black_Summer/Meshes/M_MED_Cat_Burglar_Black_Summer_jacket_dyn
bodies {
 dyn_jacket_bk [{shapeType=Capsule, length=8, radius=2, center=<3.6321 0 0>, rotation=<rot 0 0 90>}]
 dyn_jacket_bk_MSTR [{shapeType=Capsule, length=1, radius=1}]
 dyn_jacket_l_01 [{shapeType=Capsule, length=2, radius=2, center=<2.7134 0 0>, rotation=<rot 90 0 0>}]
 dyn_jacket_l_MSTR [{shapeType=Capsule, length=1, radius=1}]
 dyn_jacket_l_fr_01 [{shapeType=Capsule, length=6.5, radius=1.75, center=<3 0 0>, rotation=<rot 0 0 79.9996>}]
 dyn_jacket_l_fr_02 [{shapeType=Capsule, length=6.5, radius=1.75, center=<4.1531 0 0>, rotation=<rot -4.209 -24.6658 80.9189>}]
 dyn_jacket_l_fr_MSTR [{shapeType=Capsule, length=1, radius=1}]
 dyn_jacket_r_01 [{shapeType=Capsule, length=2, radius=2, center=<2.7134 0 0>, rotation=<rot 90 0 0>}]
 dyn_jacket_r_MSTR [{shapeType=Capsule, length=1, radius=1}]
 dyn_jacket_r_fr_01 [{shapeType=Capsule, length=6.5, radius=1.75, center=<3 0 0>, rotation=<rot 0 0 100.0004>}]
 dyn_jacket_r_fr_02 [{shapeType=Capsule, length=6.5, radius=1.75, center=<4.9689 0.3844 -0.0676>, rotation=<rot -5.7168 34.5906 98.2201>}]
 dyn_jacket_r_fr_MSTR [{shapeType=Capsule, length=1, radius=1}]
 pelvis [{shapeType=Capsule, length=6.6066, radius=8.3609, center=<5.2027 0.4138 0>}]
 thigh_twist_01_l [{shapeType=Capsule, length=28.1662, radius=7.293, center=<9.5932 -1.5419 -0.2877>, rotation=<rot 88.4573 172.7395 172.5514>}]
 thigh_twist_01_r [{shapeType=Capsule, length=28.1662, radius=7.293, center=<-9.5899 0.5064 0.2913>, rotation=<rot 88.4578 172.7395 -7.4489>}]
}
constraints [
 {bone1Name=dyn_jacket_bk, bone2Name=dyn_jacket_bk_MSTR, swing1LimitDegrees=10, swing2LimitDegrees=20, twistLimitDegrees=2}
 {bone1Name=dyn_jacket_l_fr_01, bone2Name=dyn_jacket_l_fr_MSTR, swing1LimitDegrees=10, swing2LimitDegrees=90, twistLimitDegrees=5}
 {bone1Name=dyn_jacket_l_fr_02, bone2Name=dyn_jacket_l_fr_01, swing1LimitDegrees=10, swing2LimitDegrees=90, twistLimitDegrees=5}
 {bone1Name=dyn_jacket_r_fr_01, bone2Name=dyn_jacket_r_fr_MSTR, swing1LimitDegrees=10, swing2LimitDegrees=90, twistLimitDegrees=5}
 {bone1Name=dyn_jacket_r_fr_02, bone2Name=dyn_jacket_r_fr_01, swing1LimitDegrees=10, swing2LimitDegrees=90, twistLimitDegrees=5}
 {bone1Name=dyn_jacket_l_01, bone2Name=dyn_jacket_l_MSTR, swing1LimitDegrees=10, swing2LimitDegrees=20, twistLimitDegrees=2}
 {bone1Name=dyn_jacket_r_01, bone2Name=dyn_jacket_r_MSTR, swing1LimitDegrees=10, swing2LimitDegrees=20, twistLimitDegrees=2}
 {
  bone1Name=dyn_jacket_bk_MSTR
  bone2Name=pelvis
  swing1LimitDegrees=45
  swing1Motion=Free
  swing2LimitDegrees=45
  swing2Motion=Free
  twistLimitDegrees=45
  twistMotion=Free
 }
 {
  bone1Name=dyn_jacket_r_MSTR
  bone2Name=pelvis
  swing1LimitDegrees=45
  swing1Motion=Free
  swing2LimitDegrees=45
  swing2Motion=Free
  twistLimitDegrees=45
  twistMotion=Free
 }
 {
  bone1Name=dyn_jacket_l_MSTR
  bone2Name=pelvis
  swing1LimitDegrees=45
  swing1Motion=Free
  swing2LimitDegrees=45
  swing2Motion=Free
  twistLimitDegrees=45
  twistMotion=Free
 }
 {
  bone1Name=dyn_jacket_l_fr_MSTR
  bone2Name=pelvis
  swing1LimitDegrees=45
  swing1Motion=Free
  swing2LimitDegrees=45
  swing2Motion=Free
  twistLimitDegrees=45
  twistMotion=Free
 }
 {
  bone1Name=dyn_jacket_r_fr_MSTR
  bone2Name=pelvis
  swing1LimitDegrees=45
  swing1Motion=Free
  swing2LimitDegrees=45
  swing2Motion=Free
  twistLimitDegrees=45
  twistMotion=Free
 }
 {
  bone1Name=thigh_twist_01_l
  bone2Name=pelvis
  swing1LimitDegrees=45
  swing1Motion=Free
  swing2LimitDegrees=45
  swing2Motion=Free
  twistLimitDegrees=45
  twistMotion=Free
 }
 {
  bone1Name=thigh_twist_01_r
  bone2Name=pelvis
  swing1LimitDegrees=45
  swing1Motion=Free
  swing2LimitDegrees=45
  swing2Motion=Free
  twistLimitDegrees=45
  twistMotion=Free
 }
]
