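# Physics asset for the dangling necklace on the medium male Cat Burglar Summer player body cosmetic.
# A short capsule chain runs from a small master body through two link bodies, with narrow swing and very tight twist limits between them.
# Large capsules on the upper two spine bones give the torso volume, and the spine-to-spine and master-to-spine joints leave all angular motion free.
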PhysicsAsset /BRCosmetics/Characters/Player/Male/Medium/Bodies/M_MED_Cat_Burglar_Summer/Meshes/M_MED_Cat_Burglar_Summer_dyn_necklace
bodies {
 dyn_necklace_01 [{shapeType=Capsule, length=3.875, radius=2.375, center=<4 0 0>, rotation=<rot 90 0 0>}]
 dyn_necklace_02 [{shapeType=Capsule, length=3.875, radius=2.375, center=<4 0 0>, rotation=<rot 90 0 0>}]
 dyn_necklace_MSTR [{shapeType=Capsule, length=0.5, radius=0.5}]
 spine_04 [{shapeType=Capsule, length=16.9509, radius=9.2111, center=<4.8466 0.5115 0>, rotation=<rot 0 90 0>}]
 spine_05 [{shapeType=Capsule, length=16.5759, radius=9.7111, center=<-1.744 -1.234 0>, rotation=<rot 0 90 0>}]
}
constraints [
 {
  bone1Name=dyn_necklace_MSTR
  bone2Name=spine_05
  swing1LimitDegrees=45
  swing1Motion=Free
  swing2LimitDegrees=45
  swing2Motion=Free
  twistLimitDegrees=45
  twistMotion=Free
 }
 {bone1Name=dyn_necklace_01, bone2Name=dyn_necklace_MSTR, swing1LimitDegrees=7, swing2LimitDegrees=20, twistLimitDegrees=2}
 {bone1Name=dyn_necklace_02, bone2Name=dyn_necklace_01, swing1LimitDegrees=7, swing2LimitDegrees=20, twistLimitDegrees=2}
 {
  bone1Name=spine_05
  bone2Name=spine_04
  swing1LimitDegrees=45
  swing1Motion=Free
  swing2LimitDegrees=45
  swing2Motion=Free
  twistLimitDegrees=45
  twistMotion=Free
 }
]
