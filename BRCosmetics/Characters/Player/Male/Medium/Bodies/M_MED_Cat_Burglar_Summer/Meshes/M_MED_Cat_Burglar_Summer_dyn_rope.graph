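# Physics asset for the dangling rope on the male medium Cat Burglar Summer player cosmetic body.
# Capsule bodies form left and right two-link rope chains branching from a master rope bone attached to the pelvis.
# Chain joints have swing limits and very tight twist, while the master-to-pelvis and thigh-to-pelvis constraints are free in rotation.
# Larger pelvis and thigh capsules approximate the character's hips and upper legs alongside the rope.
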PhysicsAsset /BRCosmetics/Characters/Player/Male/Medium/Bodies/M_MED_Cat_Burglar_Summer/Meshes/M_MED_Cat_Burglar_Summer_dyn_rope
bodies {
 dyn_rope_01_l [{shapeType=Capsule, length=2.5, radius=1.5, center=<2 0 1>, rotation=<rot 90 0 0>}]
 dyn_rope_01_r [{shapeType=Capsule, length=2, radius=1.5, center=<1 0 1>, rotation=<rot 90 0 0>}]
 dyn_rope_02_l [{shapeType=Capsule, length=2.25, radius=1.75, center=<2 0 1>, rotation=<rot 90 0 0>}]
 dyn_rope_02_r [{shapeType=Capsule, length=1.75, radius=1.75, center=<1 0 1>, rotation=<rot 90 0 0>}]
 dyn_rope_MSTR [{shapeType=Capsule, length=2.2767, radius=0.7796, center=<0.0955 0.0254 -0.1812>, rotation=<rot 0 -178.9218 -87.9463>}]
 pelvis [{shapeType=Capsule, length=7.4405, radius=11.7413, center=<0.9733 -3.1108 0.0002>}]
 thigh_l [{shapeType=Capsule, length=10.9405, radius=8.2413, center=<-2.7644 2.3052 0.5197>, rotation=<rot 90 0 0>}]
 thigh_r [{shapeType=Capsule, length=10.9405, radius=8.2413, center=<2.7644 -2.3052 -0.5197>, rotation=<rot 90 0 -180>}]
}
constraints [
 {
  bone1Name=dyn_rope_MSTR
  bone2Name=pelvis
  swing1LimitDegrees=45
  swing1Motion=Free
  swing2LimitDegrees=45
  swing2Motion=Free
  twistLimitDegrees=45
  twistMotion=Free
 }
 {bone1Name=dyn_rope_01_r, bone2Name=dyn_rope_MSTR, swing1LimitDegrees=45, swing2LimitDegrees=45, twistLimitDegrees=2}
 {bone1Name=dyn_rope_02_r, bone2Name=dyn_rope_01_r, swing1LimitDegrees=45, swing2LimitDegrees=45, twistLimitDegrees=2}
 {bone1Name=dyn_rope_01_l, bone2Name=dyn_rope_MSTR, swing1LimitDegrees=45, swing2LimitDegrees=45, twistLimitDegrees=2}
 {bone1Name=dyn_rope_02_l, bone2Name=dyn_rope_01_l, swing1LimitDegrees=45, swing2LimitDegrees=45, twistLimitDegrees=2}
 {
  bone1Name=thigh_l
  bone2Name=pelvis
  swing1LimitDegrees=45
  swing1Motion=Free
  swing2LimitDegrees=45
  swing2Motion=Free
  twistLimitDegrees=45
  twistMotion=Free
 }
 {
  bone1Name=thigh_r
  bone2Name=pelvis
  swing1LimitDegrees=45
  swing1Motion=Free
  swing2LimitDegrees=45
  swing2Motion=Free
  twistLimitDegrees=45
  twistMotion=Free
 }
]
